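# StateTree for an NPC boss music encounter, gating setup on the actor's initialization tags.
# Initialized requires no blocking tags and all required tags, then spawns an AudioGameplayVolume and binds NPC boss events.
# It parents an empty Paused state and a Started linked-asset state, with no transitions defined here between them.
# Uninitialized is the conditionless sibling after Initialized, ending the tree as succeeded when it completes.
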StateTree /MusicEncounter/StateTree/ST_MusicEncounter_Boss_NPC
states [
 Root {
  id=#1
  type=STATE
  children [#2, #3]
  tasks []
  transitions []
 }
 Initialized {
  id=#2
  type=STATE
  children [#4, #5]
  enter_conditions [
   {
    id=#6
    description=Actor Does Not Have ANY Initialization Blocking Tags
    properties {
     iD=#7
     instanceObject=&self:StateTreeEditorData_0.StateTreeState_12.STC_MusicEncounter_ActorHasTag_C_0
     node {
      name=Actor Does Not Have ANY Initialization Blocking Tags
      _structType=/Script/StateTreeModule.StateTreeBlueprintConditionWrapper
      conditionClass=/MusicEncounter/StateTree/Tasks/STC_MusicEncounter_ActorHasAnyMatchingTags.STC_MusicEncounter_ActorHasAnyMatchingTags_C
      deltaIndent=0
      evaluationMode=Evaluated
      operand=And
      bindingsBatch {value=65535}
      executionRuntimeTemplateIndex {value=65535}
      instanceDataHandle {index=65535, source=None, stateHandle {index=65535}}
      instanceTemplateIndex {value=65535}
      outputBindingsBatch {value=65535}
     }
    }
   }
   {
    id=#8
    description=Actor Has ALL Required Initialization Tags
    properties {
     iD=#9
     instanceObject=&self:StateTreeEditorData_0.StateTreeState_12.STC_MusicEncounter_ActorHasAllMatchingTags_C_0
     node {
      name=Actor Has ALL Required Initialization Tags
      _structType=&StateTreeBlueprintConditionWrapper
      conditionClass=/MusicEncounter/StateTree/Tasks/STC_MusicEncounter_ActorHasAllMatchingTags.STC_MusicEncounter_ActorHasAllMatchingTags_C
      deltaIndent=0
      evaluationMode=Evaluated
      operand=And
      bindingsBatch {value=65535}
      executionRuntimeTemplateIndex {value=65535}
      instanceDataHandle {index=65535, source=None, stateHandle {index=65535}}
      instanceTemplateIndex {value=65535}
      outputBindingsBatch {value=65535}
     }
    }
   }
  ]
  tasks [
   {
    id=#10
    description=Spawn AudioGameplayVolume
    properties {
     iD=#11
     instanceObject=&self:StateTreeEditorData_0.StateTreeState_12.STT_MusicEncounter_SpawnAudioGameplayVolume_C_0
     node {
      name=Spawn AudioGameplayVolume
      _structType=&StateTreeBlueprintTaskWrapper
      taskClass=/MusicEncounter/StateTree/Tasks/STT_MusicEncounter_SpawnAudioGameplayVolume.STT_MusicEncounter_SpawnAudioGameplayVolume_C
     }
    }
   }
   {
    id=#12
    description=Bind Events - NPC Boss
    properties {
     iD=#13
     instanceObject=&self:StateTreeEditorData_0.StateTreeState_12.STT_MusicEncounter_BindEvents_NPCBoss_C_0
     node {
      name=Bind Events - NPC Boss
      _structType=&StateTreeBlueprintTaskWrapper
      taskClass=/MusicEncounter/StateTree/Tasks/STT_MusicEncounter_BindEvents_NPCBoss.STT_MusicEncounter_BindEvents_NPCBoss_C
     }
    }
   }
  ]
  transitions []
 }
 Uninitialized {
  id=#3
  type=STATE
  tasks []
  transitions [{id=#14, trigger=ON_STATE_COMPLETED, state {id=#15, name=None, linkType=SUCCEEDED}}]
 }
 Paused {id=#4, type=STATE, tasks [], transitions []}
 Started {id=#5, type=LINKED_ASSET, transitions []}
]
