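# StateTree for the music of an NPC boss in the MusicEncounter content, currently a skeleton that lays out the music phases.
# A Root state holds three child states: Idle, Combat and Ending.
# No state has tasks or transitions, so the tree defines no music behavior or switching between phases.
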StateTree /MusicEncounter/StateTree/ST_MusicEncounter_NPCBoss_Music
states [
 Root {
  id=#1
  type=STATE
  children [#2, #3, #4]
  tasks []
  transitions []
 }
 Idle {id=#2, type=STATE, tasks [], transitions []}
 Combat {id=#3, type=STATE, tasks [], transitions []}
 Ending {id=#4, type=STATE, tasks [], transitions []}
]
